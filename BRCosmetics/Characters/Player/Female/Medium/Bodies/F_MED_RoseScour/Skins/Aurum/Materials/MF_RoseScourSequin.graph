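# Material function for the RoseScour Aurum skin that layers a tinted sequin look onto incoming material attributes.
# Tiled roughness and normal textures shape the sequins, while Fresnel math modulates a color built from falloff tints and optional base color.
# The Mask input blends base color, metallic, normal and switchable sparkle emissive, but the roughness blend has no mask connected.
# Parameters choose between sequin and base values for metalness, roughness and normals.
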MaterialFunction /BRCosmetics/Characters/Player/Female/Medium/Bodies/F_MED_RoseScour/Skins/Aurum/Materials/MF_RoseScourSequin
graph {nodes [
  n1 = FunctionInput(Preview=SetMaterialAttributes!(MaterialAttributes)) {bUsePreviewValueAsDefault=true, inputName=inMat, inputType=FunctionInput_MaterialAttributes}
  n2 = GetMaterialAttributes(None=@n1) {attributeGetTypes [
    #1
    #2
    #3
    #4
    #5
   ]}
  n3 = GetMaterialAttributes(None=@n1) {attributeGetTypes [
    #3
    #4
    #1
    #2
    #5
   ]}
  n4 = FunctionInput(Preview=Constant3Vector()) {bUsePreviewValueAsDefault=true, inputName=Mask, inputType=FunctionInput_Scalar}
  n5 = ScalarParameter() {defaultValue=1.5, group=MF_Sequin, parameterName=SequinBrightness}
  n6 = Fresnel(ExponentIn=ScalarParameter() {defaultValue=1, group=MF_Sequin, parameterName=SequinFresnel}) {exponent=1}
  n7 = Multiply(A=@n6, B=@n6)
  n8 = OneMinus(None=@n7)
  n9 = Multiply(A=NamedRerouteUsage(), B=LinearInterpolate(Alpha=Multiply(A=@n8, B=@n7)) {constA=0.7})
  n10 = ScalarParameter() {defaultValue=0.9, group=MF_Sequin, parameterName=SequinMetalness}
  n11 = NamedRerouteUsage()
  n12 = ScalarParameter() {group=MF_Sequin, parameterName=UseBaseMetalness}
  n13 = NamedRerouteUsage()
  n14 = TextureSampleParameter2D(UVs=NamedRerouteUsage()) {
   group=MF_Sequin
   parameterName=Sequin_TilingRoughness
   samplerSource=SSM_Wrap_WorldGroupSettings
   samplerType=SAMPLERTYPE_LinearGrayscale
   texture=/Game/Global/Textures/Default/Tileable_Patterns/T_SequinTile_roughness
  }
  n15 = LinearInterpolate(
   A=ScalarParameter() {group=MF_Sequin, parameterName=Sequin_MinRoughness}
   B=ScalarParameter() {defaultValue=1, group=MF_Sequin, parameterName=Sequin_MaxRoughness}
   Alpha=@n14.G
  )
  n16 = NamedRerouteUsage()
  n17 = TextureSampleParameter2D(UVs=NamedRerouteUsage()) {
   group=MF_Sequin
   parameterName=Sequin_TilingNormal
   samplerSource=SSM_Wrap_WorldGroupSettings
   samplerType=SAMPLERTYPE_Normal
   texture=/Game/Global/Textures/Default/Tileable_Patterns/T_SequinTile_N
  }
  n18 = Multiply(A=ScalarParameter() {defaultValue=0.8, group=MF_Sequin, parameterName=SequinNormalIntensity}, B=Constant3Vector() {constant=<rgba 1 1 0 0>})
  n19 = Add(A=@n16, B=Multiply(A=@n17.RGB, B=@n18)) {constB=0}
  n20 = NamedRerouteUsage()
  n21 = ScalarParameter() {defaultValue=1, group=MF_Sequin, parameterName=SequinEmissiveIntensity}
  n22 = NamedRerouteUsage()
  n23 = StaticSwitchParameter(True=Add(A=Multiply(A=@n20, B=@n21), B=@n22), False=Add(A=NamedRerouteUsage(), B=@n22)) {desc=Replace emissive channel for sequin color, group=MF_Sequin, parameterName=EnableEmissiveOverride}
  n24 = SetMaterialAttributes!(
   MaterialAttributes=@n2.MaterialAttributes
   BaseColor=LinearInterpolate(A=@n3.BaseColor, B=Multiply(A=@n5, B=@n9), Alpha=@n4)
   Metallic=LinearInterpolate(A=@n3.Metallic, B=LinearInterpolate(A=@n10, B=@n11, Alpha=@n12), Alpha=@n4)
   Roughness=LinearInterpolate(A=@n3.Roughness, B=StaticSwitchParameter(True=@n13, False=@n15) {group=MF_Sequin, parameterName=UseBaseRoughness})
   Normal=LinearInterpolate(A=@n3.Normal, B=StaticSwitchParameter(True=@n16, False=@n19) {group=MF_Sequin, parameterName=UseBaseNormal}, Alpha=@n4)
   EmissiveColor=StaticSwitchParameter(True=LinearInterpolate(A=@n3.EmissiveColor, B=@n23, Alpha=@n4), False=@n3.EmissiveColor) {group=MF_Sequin, parameterName=EnableSequinSparkle}
  ) {attributeSetTypes [
    #3
    #4
    #1
    #2
    #5
   ]}
  FunctionOutput(None=@n24)
  Multiply(A=@n16, B=@n18)
  NamedRerouteDeclaration(None=Multiply(A=TextureCoordinate(), B=ScalarParameter() {defaultValue=32, group=MF_Sequin, parameterName=SequinTile})) {name=SequinUVs, nodeColor=<rgba 0 1 0.0859 1>}
  n25 = VectorParameter() {group=MF_Sequin, parameterName=SequinFalloffColor01, defaultValue=<rgba 0.8842 0.6282 1 1>}
  n26 = VectorParameter() {group=MF_Sequin, parameterName=SequinFalloffColor02, defaultValue=<rgba 0.6663 0.575 1 1>}
  n27 = Multiply(B=@n8)
  n28 = NamedRerouteUsage()
  n29 = Multiply(A=@n28, B=@n25.RGB)
  n30 = Multiply(A=@n28, B=@n26.RGB)
  NamedRerouteDeclaration(None=LinearInterpolate(
    A=LinearInterpolate(A=@n25.RGB, B=@n26.RGB, Alpha=@n27)
    B=LinearInterpolate(A=@n29, B=@n30, Alpha=@n27)
    Alpha=ScalarParameter() {defaultValue=1, group=MF_Sequin, parameterName=UseBaseColor}
   )) {name=SequinColor, nodeColor=<rgba 0.8984 0 1 1>}
  NamedRerouteDeclaration(None=@n2.Roughness) {name=BaseRoughness, nodeColor=<rgba 0.6172 1 0 1>}
  NamedRerouteDeclaration(None=@n2.Normal) {name=BaseNormals, nodeColor=<rgba 0 1 0.1562 1>}
  NamedRerouteDeclaration(None=@n2.BaseColor) {name=BaseColor, nodeColor=<rgba 0 1 0.7188 1>}
  NamedRerouteDeclaration(None=@n2.Metallic) {name=BaseMetalness, nodeColor=<rgba 0.6406 0 1 1>}
  NamedRerouteDeclaration(None=@n2.EmissiveColor) {name=BaseEmissiveness, nodeColor=<rgba 0 0.7422 1 1>}
 ]}
