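# Material function for the RoseScour Aurum female skin, adding gold emissive, animated glitter and thin-film color to incoming material attributes.
# It runs MF_RoseScourSequin, then layers masked glitter from depth-faded, parallax sparkle samples and panning noise.
# Pre-skinned position masks confine effects to lower legs, UV steps place tail glitter, and IsBackpack blends backpack behavior.
# A Fresnel-driven spectrum lookup tints base color, while quality switches simplify costlier samples on lower tiers.
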MaterialFunction /BRCosmetics/Characters/Player/Female/Medium/Bodies/F_MED_RoseScour/Skins/Aurum/Materials/MF_RoseScourAurum
graph {nodes [
  FunctionOutput(None=NamedRerouteUsage())
  n1 = TextureSample(
   UVs=Multiply(A=TextureCoordinate(), B=ScalarParameter() {defaultValue=2, group=Thin Film_2, parameterName=UvFilm2OffsetTiling})
   Tex=NamedRerouteUsage()
  ) {samplerSource=SSM_Wrap_WorldGroupSettings, texture=/Game/Effects/Fort_Effects/Textures/Tile/T_PerlinRGB_01_M}
  n2 = Multiply(A=@n1.B) {constB=0.1}
  n3 = Constant()
  NamedRerouteDeclaration(None=MaterialFunctionCall!(Low=@n2, High=@n2, Epic=@n2, --------, Mobile=@n3, Medium=@n3) {materialFunction=/Game/Functions/MF_QualitySwitch_Scalar}.Result) {name=UVOffset, nodeColor=<rgba 0.4062 1 0 1>}
  n4 = Add(
   A=Multiply(A=TextureCoordinate() {uTiling=1.5}, B=ScalarParameter() {defaultValue=4, group=EmissiveGold, parameterName=GoldEmissiveNoiseSize}) {constB=4}
   B=Multiply(A=Time(), B=Constant2Vector() {g=0.15, r=0.05}) {constB=0.3}
  )
  n5 = Fresnel(
   ExponentIn=ScalarParameter() {defaultValue=2, group=EmissiveGold, parameterName=ExponentIn, sortPriority=31}
   BaseReflectFractionIn=ScalarParameter() {defaultValue=0.05, group=EmissiveGold, parameterName=BaseReflectFractionIn}
  ) {exponent=3}
  n6 = Saturate(None=Subtract(
    A=TextureSample(UVs=@n4) {samplerSource=SSM_Wrap_WorldGroupSettings, texture=/Game/Effects/Fort_Effects/Textures/Tile/T_Inky_Smoke_Tile}.R
    B=Power(Base=@n5)
   ))
  n7 = Multiply(A=@n6, B=NamedRerouteUsage()) {constA=1}
  NamedRerouteDeclaration(None=SmoothStep(Value=@n7) {constMax=0.75}) {name=DarkEmissiveLowerLeg, nodeColor=<rgba 0.6641 0 1 1>}
  n8 = Add(A=NamedRerouteUsage()) {constB=0.16}
  n9 = VectorParameter() {group=EmissiveGold, parameterName=GoldRosesColor, defaultValue=<rgba 0.6377 0.326 0.0647 1>}
  n10 = SmoothStep(Value=NamedRerouteUsage()) {constMin=0.95}
  n11 = SmoothStep(Value=NamedRerouteUsage()) {constMax=0.75}
  n12 = Subtract(A=@n11, B=Floor(None=@n11))
  n13 = Multiply(A=@n12, B=NamedRerouteUsage())
  n14 = Multiply(A=Saturate(None=OneMinus(None=@n10)), B=Add(A=NamedRerouteUsage(), B=Add(A=@n12, B=@n13)))
  NamedRerouteDeclaration(None=LinearInterpolate(A=NamedRerouteUsage(), B=Multiply(A=@n8, B=@n9.RGB), Alpha=Saturate(None=@n14)) {constA=1, constAlpha=0, constB=0}) {name=Emissive_PreGlitter, nodeColor=<rgba 1 0 0.5859 1>}
  n15 = FunctionInput() {inputName=TechArtMask, sortPriority=1}
  NamedRerouteDeclaration(None=ComponentMask(None=@n15) {r=true}) {name=FX1_EmissiveMask, nodeColor=<rgba 1 0 0 1>}
  n16 = NamedRerouteUsage()
  n17 = NamedRerouteUsage()
  n18 = NamedRerouteUsage()
  n19 = Add(A=TextureCoordinate() {uTiling=-1.5, vTiling=0.5}, B=NamedRerouteUsage())
  n20 = ScalarParameter() {defaultValue=3, group=Glitter, parameterName=GlitterLowerLegMaskSize}
  n21 = Time()
  n22 = Constant2Vector() {g=0.03, r=0.05}
  n23 = TextureSample(UVs=Add(A=Multiply(A=@n19, B=@n20) {constB=3}, B=Multiply(A=@n21, B=@n22) {constB=0.3})) {
   samplerSource=SSM_Wrap_WorldGroupSettings
   samplerType=SAMPLERTYPE_LinearGrayscale
   texture=/Game/Effects/Fort_Effects/Textures/Tile/T_Caustic_Wind_Diagonal
  }
  n24 = Multiply(A=@n16, B=Multiply(A=Multiply(A=@n17, B=@n18), B=SmoothStep(Value=@n23.R) {constMax=0.5, constMin=0.35}))
  n25 = ScalarParameter() {defaultValue=1, group=Glitter, parameterName=GlitterLowerLegToggle}
  n26 = OneMinus(None=MaterialFunctionCall(Exponent=ScalarParameter() {defaultValue=3, group=Glitter, parameterName=NoiseFresnelExponent}) {materialFunction=&AxisAlignedFresnel}.Fresnel)
  n27 = Saturate(None=Power(Base=@n26, Exp=ScalarParameter() {defaultValue=2, group=Glitter, parameterName=BackpackFresnelPow}) {constExponent=1.5})
  NamedRerouteDeclaration(None=LinearInterpolate(A=Multiply(A=@n24, B=@n25), B=Multiply(A=@n27, B=@n16), Alpha=ScalarParameter() {parameterName=IsBackpack})) {name=SpinningGlitterMask, nodeColor=<rgba 0.875 1 0 1>}
  NamedRerouteDeclaration(None=TextureObject() {texture=&T_PerlinRGB_01_M}) {name=UVOffsetTex, nodeColor=<rgba 1 0 0.7969 1>}
  NamedRerouteDeclaration(None=ComponentMask(None=@n15) {g=true}) {name=JacketMask, nodeColor=<rgba 0.8047 1 0 1>}
  n28 = GetMaterialAttributes(None=NamedRerouteUsage()) {attributeGetTypes [#1, #2]}
  n29 = NamedRerouteDeclaration(None=@n28.MaterialAttributes) {name=Material Attributes, nodeColor=<rgba 0.0781 0 1 1>}
  n30 = Constant()
  n31 = Multiply(
   A=MaterialFunctionCall(Low=@n1.R, High=@n1.R, Epic=@n1.R, Mobile=@n30, Medium=@n30) {materialFunction=&MF_QualitySwitch_Scalar}.Result
   B=ScalarParameter() {defaultValue=0.2, group=Thin Film_2, parameterName=ThinFilm2UVOffsetIntensity}
  )
  n32 = FunctionInput() {inputName=Fresnel, inputType=FunctionInput_Scalar}
  n33 = ScalarParameter() {defaultValue=0.5, group=Thin Film_2, parameterName=ThinFilm2Warp, sortPriority=4}
  n34 = Add(
   A=Multiply(A=ScalarParameter() {defaultValue=1, group=Thin Film_2, parameterName=ThinFilm2Scale, sortPriority=5}, B=Power(Base=@n32, Exp=@n33))
   B=ScalarParameter() {group=Thin Film_2, parameterName=ThinFilm2Offset}
  )
  n35 = TextureSampleParameter2D(UVs=StaticSwitchParameter(True=Add(A=@n31, B=@n34), False=@n34) {group=Thin Film_2, parameterName=UseThinFilmUVOffset}) {
   group=Thin Film_2
   parameterName=Thin Film 2 Texture
   samplerSource=SSM_Wrap_WorldGroupSettings
   sortPriority=2
   texture=/Game/Global/Textures/Default/Thinfilm/T_ThinFilm_Spectrum_COLOR
  }
  n36 = ScalarParameter() {defaultValue=0.5, group=Thin Film_2, parameterName=ThinFilm2Intensity, sortPriority=2}
  n37 = Power(Base=@n32, Exp=ScalarParameter() {defaultValue=2, group=Thin Film_2, parameterName=ThinFilm2Exponent, sortPriority=3})
  n38 = Multiply(A=LinearInterpolate(Alpha=NamedRerouteUsage()) {constA=1, constB=0.01}, B=@n37)
  n39 = ScalarParameter() {group=Thin Film_2, parameterName=RoughnessInfluence2}
  n40 = LinearInterpolate(
   A=@n28.BaseColor
   B=Add(A=@n28.BaseColor, B=Multiply(A=@n35.RGB, B=@n36))
   Alpha=Saturate(None=LinearInterpolate(A=@n37, B=@n38, Alpha=@n39))
  )
  NamedRerouteDeclaration(None=BlendMaterialAttributes(
    A=NamedRerouteUsage()
    B=SetMaterialAttributes!(MaterialAttributes=@n29, BaseColor=@n40) {attributeSetTypes [#2]}
    Alpha=NamedRerouteUsage()
   ) {vertexAttributeBlendType=UseA}) {name=AfterThinFilm2, nodeColor=<rgba 1 0.2344 0 1>}
  n41 = NamedRerouteUsage()
  n42 = NamedRerouteDeclaration(None=FunctionInput() {inputName=Material, inputType=FunctionInput_MaterialAttributes}) {name=MA_IN, nodeColor=<rgba 1 0.9141 0 1>}
  n43 = GetMaterialAttributes(None=BlendMaterialAttributes(
    A=MaterialFunctionCall!(Mask=@n41, inMat=@n42) {materialFunction=&/MF_RoseScourSequin}.Result
    B=@n42
    Alpha=ScalarParameter() {parameterName=IsBackpack}
   )) {attributeGetTypes [#2, #3]}
  n44 = Multiply(A=NamedRerouteUsage(), B=NamedRerouteUsage())
  n45 = NamedRerouteUsage()
  n46 = NamedRerouteUsage()
  n47 = Multiply(A=Saturate(None=Subtract(A=@n45, B=@n46)), B=NamedRerouteUsage())
  n48 = NamedRerouteUsage()
  n49 = NamedRerouteUsage()
  n50 = LinearInterpolate(A=LinearInterpolate(A=Add(A=@n43.BaseColor, B=@n44), B=NamedRerouteUsage(), Alpha=@n47), B=@n48, Alpha=@n49)
  n51 = Multiply(A=NamedRerouteUsage(), B=ScalarParameter() {defaultValue=1, parameterName=HasTailGlitter})
  n52 = Multiply(A=@n43.EmissiveColor) {constB=0.5}
  n53 = ScalarParameter() {parameterName=IsBackpack}
  n54 = Multiply(A=@n47, B=NamedRerouteUsage())
  n55 = LinearInterpolate(A=Add(A=Multiply(A=@n52, B=@n53), B=Add(A=@n54, B=@n54)), B=@n48, Alpha=@n49)
  NamedRerouteDeclaration(None=SetMaterialAttributes!(MaterialAttributes=@n43.MaterialAttributes, BaseColor=Add(A=@n50, B=@n51), EmissiveColor=Add(A=@n55, B=@n51)) {attributeSetTypes [#2, #3]}) {name=BeforeThinFilm2, nodeColor=<rgba 0 0.0859 1 1>}
  n56 = VertexInterpolator(VS=LocalPosition() {includedOffsets=ExcludeOffsets, localOrigin=InstancePreSkinning}.XYZ)
  n57 = ComponentMask(None=@n56.PS) {b=true}
  n58 = ScalarParameter() {defaultValue=0.23, group=EmissiveGold, parameterName=PosMaskPowLeft}
  n59 = ScalarParameter() {defaultValue=0.2, group=EmissiveGold, parameterName=PosMaskPowRight}
  n60 = Saturate(None=ComponentMask(None=@n56.PS) {r=true})
  n61 = Subtract(
   A=Power(Base=@n57, Exp=LinearInterpolate(A=@n58, B=@n59, Alpha=@n60)) {constExponent=0.3}
   B=ScalarParameter() {defaultValue=1.4083, group=EmissiveGold, parameterName=PosMaskHeight}
  ) {constB=3.2}
  NamedRerouteDeclaration(None=Saturate(None=OneMinus(None=@n61))) {name=GoldLowerLegMask, nodeColor=<rgba 1 0.4101 0 1>}
  n62 = VectorParameter() {group=EmissiveGold, parameterName=GoldRosesColor, defaultValue=<rgba 0.6377 0.326 0.0647 1>}
  n63 = Multiply(A=@n62.RGB) {constB=0.33}
  n64 = Multiply(A=@n62.RGB) {constB=8}
  NamedRerouteDeclaration(None=Multiply(A=LinearInterpolate(A=@n63, B=@n64, Alpha=@n7), B=NamedRerouteUsage())) {name=GoldEmissive, nodeColor=<rgba 1 0.3123 0 1>}
  n65 = Subtract(A=Power(Base=@n57, Exp=@n59) {constExponent=0.3}, B=Constant() {r=1.25}) {constB=3.2}
  NamedRerouteDeclaration(None=Saturate(None=OneMinus(None=@n65))) {name=GlitterLowerLegMask, nodeColor=<rgba 1 1 1 1>}
  n66 = Multiply(A=Multiply(A=NamedRerouteUsage(), B=NamedRerouteUsage()), B=NamedRerouteUsage())
  n67 = NamedRerouteUsage()
  n68 = Power(
   Base=Multiply(A=@n66, B=LinearInterpolate(Alpha=@n67) {constA=0.5})
   Exp=ScalarParameter() {defaultValue=4, group=Glitter, parameterName=GlitterPower}
  ) {constExponent=4}
  n69 = ScalarParameter() {defaultValue=100, group=Glitter, parameterName=GlitterIntensity}
  n70 = Multiply(A=NamedRerouteUsage(), B=Saturate(None=Multiply(A=@n68, B=@n69) {constB=4}))
  n71 = NamedRerouteUsage()
  n72 = Multiply(
   A=Saturate(None=Subtract(A=@n70, B=@n71))
   B=ScalarParameter() {defaultValue=10, group=Glitter, parameterName=Emissive Intensity, sliderMax=20, sortPriority=5}
  )
  n73 = Power(Base=Fresnel() {baseReflectFraction=0, exponent=1}, Exp=ScalarParameter() {defaultValue=0.5, group=Glitter, parameterName=GlitterFresnelExp}) {constExponent=3}
  n74 = NamedRerouteUsage()
  n75 = Multiply(A=Power(Base=Multiply(A=@n66, B=@n74)) {constExponent=0.5}) {constB=1.3}
  n76 = NamedRerouteUsage()
  n77 = Saturate(None=Multiply(A=NamedRerouteUsage(), B=@n76) {constB=0.5})
  NamedRerouteDeclaration(None=Add(A=Multiply(A=@n72, B=@n73), B=Add(A=@n75, B=@n77))) {name=Glitter Animated Emissive, nodeColor=<rgba 1 0.2874 0 1>}
  n78 = LinearInterpolate(A=TextureCoordinate(), B=TextureCoordinate() {uTiling=2, vTiling=2}, Alpha=Step(X=NamedRerouteUsage()) {constY=0.2})
  n79 = ScalarParameter() {defaultValue=12, group=Glitter, parameterName=Sparkle Detail Tiling, sliderMax=64, sliderMin=8, sortPriority=2}
  n80 = ScalarParameter() {defaultValue=1, group=Glitter, parameterName=Glitter Tiling Close}
  n81 = ScalarParameter() {defaultValue=0.5, group=Glitter, parameterName=Glitter Tiling Far}
  n82 = MaterialFunctionCall(
   Fade Length=ScalarParameter() {defaultValue=2, group=Glitter, parameterName=Glitter Fade Length}
   Fade Offset=ScalarParameter() {defaultValue=100, group=Glitter, parameterName=Glitter Fade Offset}
  ) {materialFunction=&CameraDepthFade}
  n83 = NamedRerouteDeclaration(None=Multiply(A=Multiply(A=@n78, B=@n79) {constB=3}, B=LinearInterpolate(A=@n80, B=@n81, Alpha=@n82.Result))) {name=UVs, nodeColor=<rgba 1 0.5859 0 1>}
  n84 = TextureSampleParameter2D(UVs=@n83) {
   group=Glitter
   parameterName=Sparkle Mask
   samplerSource=SSM_Wrap_WorldGroupSettings
   samplerType=SAMPLERTYPE_Masks
   sortPriority=0
   texture=/Game/Weapons/WeaponSkins/Textures/T_SparkleMask
  }
  n85 = TextureSampleParameter2D(UVs=BumpOffset(Coordinate=@n83, Height=ScalarParameter() {group=Glitter, parameterName=Glitter Depth})) {
   group=Glitter
   parameterName=Sparkle Mask
   samplerSource=SSM_Wrap_WorldGroupSettings
   samplerType=SAMPLERTYPE_Masks
   sortPriority=0
   texture=&T_SparkleMask
  }
  n86 = Multiply(A=@n84.R, B=@n85.R)
  NamedRerouteDeclaration(None=MaterialFunctionCall(Low=@n86, High=@n86, Epic=@n86, Mobile=Multiply(A=@n84.R) {constB=0.5}, Medium=@n86) {materialFunction=&MF_QualitySwitch_Scalar}.Result) {name=Glitter, nodeColor=<rgba 1 0 0.7734 1>}
  NamedRerouteDeclaration(None=@n79) {name=Tiling, nodeColor=<rgba 0.2656 0 1 1>}
  n87 = NamedRerouteUsage()
  n88 = Subtract(A=NamedRerouteUsage(), B=Frac(None=NamedRerouteUsage()))
  n89 = NamedRerouteUsage()
  n90 = Multiply(
   A=Add(A=Divide(A=@n88, B=@n89), B=Time())
   B=Multiply(
    A=ScalarParameter() {defaultValue=0.2, group=Glitter, parameterName=Time Scale, sliderMax=2, sortPriority=3}
    B=ScalarParameter() {defaultValue=2, group=Glitter, parameterName=SpeedGlitter}
   )
  ) {constB=0.5}
  n91 = MaterialFunctionCall(UV Channel=Frac(None=Add(A=@n87, B=@n90))) {materialFunction=&LinearGradient}
  NamedRerouteDeclaration(None=ConstantBiasScale(None=Sine(None=@n91.VGradient))) {name=GlitterAnimatedMask, nodeColor=<rgba 0.1016 0 1 1>}
  NamedRerouteDeclaration(None=MaterialFunctionCall(Low=@n85.G, High=@n85.G, Epic=@n85.G, Mobile=Multiply(A=@n84.G) {constB=0.5}, Medium=@n85.G) {materialFunction=&MF_QualitySwitch_Scalar}.Result) {name=SparkleMask, nodeColor=<rgba 1 0.3047 0 1>}
  n92 = Multiply(A=Multiply(A=NamedRerouteUsage(), B=NamedRerouteUsage()), B=Saturate(None=@n26))
  n93 = Multiply(A=NamedRerouteUsage()) {constB=2}
  NamedRerouteDeclaration(None=Saturate(None=Subtract(A=@n92, B=@n93))) {name=NoiseFresnel, nodeColor=<rgba 0 0.8594 1 1>}
  n94 = LocalPosition() {includedOffsets=ExcludeOffsets, localOrigin=InstancePreSkinning}
  n95 = Constant2Vector() {g=0.03}
  n96 = SmoothStep(Value=MaterialFunctionCall!(
    Export Float 4
    TextureObject=NamedRerouteUsage()
    TextureSize=VectorParameter() {group=Glitter, parameterName="TextureSize (V3)", defaultValue=<rgba 64 64 256 0>}.RGB
    WorldPosition=VertexInterpolator(VS=@n94.XYZ).PS
    World Space Normal
    ProjectionTransitionContrast
    Pan_YZ=@n95
    Pan_XZ=@n95
    Pan_XY
   ) {materialFunction=/BRCosmetics/Gadgets/Assets/VinderTech_GliderChute/Glider_MissusMind/Materials/WorldAlignedTexture_ScalePanXYZ}.XYZ Texture) {constMax=0.5, constMin=0.1}
  n97 = Saturate(None=Add(A=ComponentMask(None=@n96) {r=true}, B=ComponentMask(None=@n96) {g=true}))
  NamedRerouteDeclaration(None=MaterialFunctionCall(Low=@n97, High=@n97, Epic=@n97, Mobile=Constant3Vector() {constant=<rgba 1 1 1 0>}, Medium=@n97) {materialFunction=&MF_QualitySwitch_Scalar}.Result) {name=TriplanarNoise, nodeColor=<rgba 0.0078 0 1 1>}
  n98 = Constant() {r=0.75}
  NamedRerouteDeclaration(None=MaterialFunctionCall(Value_Morning=@n98, Value_Day=Constant() {r=1}, Value_Evening=@n98, Value_Night=Constant() {r=0.5}) {materialFunction=&Scalar_Time_of_Day}.Result) {name=TOD_Scalar, nodeColor=<rgba 0 1 0.6016 1>}
  n99 = TextureObjectParameter() {
   group=Glitter
   parameterName=RoseScourSequinMask
   texture=/BRCosmetics/Characters/Player/Female/Medium/Bodies/F_MED_RoseScour/Skins/Aurum/Textures/T_RoseScour_Aurum_Body_SequinMask
  }
  n100 = TextureSample(Tex=@n99) {samplerSource=SSM_Wrap_WorldGroupSettings}
  NamedRerouteDeclaration(None=@n100.R) {name=SequinMask, nodeColor=<rgba 0.1016 1 0 1>}
  NamedRerouteDeclaration(None=@n6) {name=LegNoise, nodeColor=<rgba 1 0 0.4219 1>}
  NamedRerouteDeclaration(None=@n28.Roughness) {name=RoughnessBeforeThinFilm2, nodeColor=<rgba 0 0.7422 1 1>}
  NamedRerouteDeclaration(None=VectorParameter() {group=Glitter, parameterName=GlitterColor, defaultValue=<rgba 2 0.9444 0.2882 1>}.RGB) {name=GlitterColor, nodeColor=<rgba 1 0.9141 0 1>}
  n101 = Multiply(A=@n74, B=ScalarParameter() {defaultValue=0.2, group=Glitter, parameterName=LowerLegGlitterStrength}) {constB=0.2}
  n102 = Multiply(A=@n76, B=ScalarParameter() {defaultValue=1, group=Glitter, parameterName=FresnelNoiseGlitterStrength}) {constB=0.45}
  NamedRerouteDeclaration(None=Saturate(None=Add(A=@n101, B=@n102))) {name=NoiseEmissiveMask, nodeColor=<rgba 0.2422 0 1 1>}
  NamedRerouteDeclaration(None=VectorParameter() {group=Glitter, parameterName=EmissiveNoiseColor, defaultValue=<rgba 5 2.1913 0.9115 1>}.RGB) {name=EmissiveNoiseColor, nodeColor=<rgba 0.0781 1 0 1>}
  NamedRerouteDeclaration(None=@n100.G) {name=NeckMask, nodeColor=<rgba 0.3828 0 1 1>}
  n103 = TextureCoordinate()
  n104 = ComponentMask(None=@n103) {r=true}
  n105 = Add(A=ComponentMask(None=@n103) {g=true}) {constB=-0.19}
  n106 = Step(Y=Add(A=Multiply(A=@n104) {constB=2.268}, B=Multiply(A=@n105))) {constX=0.5, constY=0.5}
  n107 = NamedRerouteUsage()
  n108 = TextureCoordinate()
  n109 = Subtract(A=Step(Y=ComponentMask(None=@n108) {r=true}) {constX=0.865}, B=Step(Y=ComponentMask(None=@n108) {r=true}) {constX=0.74})
  n110 = Multiply(
   A=Add(A=Multiply(A=@n106, B=@n107) {constB=0.1}, B=Multiply(A=@n109, B=@n107) {constB=0.1})
   B=Multiply(A=NamedRerouteUsage(), B=NamedRerouteUsage())
  )
  NamedRerouteDeclaration(None=Multiply(A=Multiply(A=@n110) {constB=2}, B=NamedRerouteUsage())) {name=TailGlitter, nodeColor=<rgba 1 0 0.375 1>}
  QualitySwitch()
  n111 = Saturate(None=Subtract(A=@n100.R, B=@n100.R))
  TemporalResponsivenessOutput!(None=Multiply(A=NamedRerouteDeclaration(None=@n111) {name=GlitterChannel, nodeColor=<rgba 0 0.4609 1 1>}) {constB=0.6})
 ]}
